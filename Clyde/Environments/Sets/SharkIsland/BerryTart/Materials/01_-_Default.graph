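# Placeholder-style default material in the BerryTart folder of the SharkIsland environment set.
# Its graph holds a single vector parameter with a dark grey default, with no textures, math or visible connection to a material output.
# It uses the Default Lit shading model, is flagged for use with Nanite meshes and keeps legacy pixel depth offset mode.
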Material /Clyde/Environments/Sets/SharkIsland/BerryTart/Materials/01_-_Default
graph {
 properties {bUsedWithNanite=true, pixelDepthOffsetMode=PDOM_Legacy, usedShadingModels=Default Lit}
 nodes [VectorParameter() {defaultValue=<rgba 0.1647 0.1647 0.1647 0>}]
}
