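# Material function for Shark Island BerryTart gems, faking refractive, sparkling diamonds over incoming material attributes.
# A refracted, parallax-offset normal map, perturbed by view-dependent rotation, forms the tinted interior.
# This combines with view-angle-tinted skylight reflections, then gets gap-brightness shading, a low-quality multiplier and eye-adaptation compensation.
# A TechArtMask channel blends the result in as emissive, zeroing base color, metallic, specular and roughness.
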MaterialFunction /Clyde/Environments/Sets/SharkIsland/BerryTart/Materials/MF_BerryTart_Gems
graph {nodes [
  Constant() {r=1}
  n1 = GetMaterialAttributes(None=FunctionInput() {inputName=Material, inputType=FunctionInput_MaterialAttributes}) {attributeGetTypes [
    #1
    #2
    #3
    #4
    #5
    #6
   ]}
  n2 = MaterialFunctionCall(Float3=NamedRerouteUsage()) {materialFunction=&BreakOutFloat3Components}
  n3 = InverseLinearInterpolate(
   A=ScalarParameter() {defaultValue=0.3, group=Diamond_Base, parameterName=Gap Brightness Min}
   B=ScalarParameter() {defaultValue=0.7, group=Diamond_Base, parameterName=Gap Brightness Max}
   Value=Max(A=Max(A=@n2.R, B=@n2.R), B=@n2.B)
  ) {constA=0.7, constB=0.8}
  n4 = Constant3Vector() {constant=<rgba 1 0 0 0>}
  n5 = TextureCoordinate()
  n6 = Multiply(
   A=Length(None=ComponentMask(None=@n5) {r=true}) {is=class MaterialExpressionLength}
   B=ScalarParameter() {defaultValue=7.5, group=Diamond_Base, parameterName=Normal Randomization Frequency}
  ) {constB=12}
  n7 = VertexNormalWS()
  n8 = CameraVectorWS()
  n9 = Multiply(A=Power(Base=DotProduct(A=@n7, B=@n8)) {constExponent=1.5}) {constB=2}
  n10 = TextureCoordinate()
  n11 = AppendVector(
   A=ScalarParameter() {defaultValue=20, group=Diamond_Base, parameterName=Diamond Internal Tiling X}
   B=ScalarParameter() {defaultValue=15, group=Diamond_Base, parameterName=Diamond Internal Tiling Y}
  )
  n12 = NamedRerouteUsage()
  n13 = VertexTangentWS()
  n14 = VertexNormalWS()
  n15 = CrossProduct(A=@n14, B=@n13)
  n16 = MaterialFunctionCall!(
   VectorToTransform=MaterialFunctionCall!(
    Ray Direction=CameraVectorWS()
    Surface Normal=Transform(None=@n12)
    Refractive Index Origin=Constant() {r=1}
    Refractive Index Target=ScalarParameter() {defaultValue=2.5, group=Diamond_Base, parameterName=Diamond Surface IOR}
   ) {materialFunction=/Engine/Functions/Engine_MaterialFunctions01/Vectors/Refract}.Vector
   BasisX=@n13
   BasisY=@n15
   BasisZ=@n14
  ) {materialFunction=/Engine/Functions/Engine_MaterialFunctions02/Math/InverseTransformMatrix}
  n17 = ScalarParameter() {defaultValue=0.3, group=Diamond_Base, parameterName=Diamond Internal Depth}
  n18 = Divide(A=Divide(A=Multiply(A=@n16.Transformed Vector 3, B=@n17), B=ComponentMask(None=@n16.Transformed Vector 3) {b=true})) {constB=1}
  n19 = TextureObjectParameter() {
   parameterName=Diamond Internal Normal
   samplerType=SAMPLERTYPE_Normal
   texture=/Clyde/Environments/Sets/SharkIsland/BerryTart/Textures/T_BerryTart_Diamond_N
  }
  n20 = TextureSample(UVs=Add(A=Multiply(A=@n10, B=@n11), B=ComponentMask(None=@n18) {g=true, r=true}), Tex=@n19) {samplerType=SAMPLERTYPE_Normal, texture=/ClydeCosmetics/Weapons/FlowGrappler/Materials/T_BerryTart_Diamond_N}
  n21 = Divide(B=ScalarParameter() {defaultValue=3, group=Diamond_Base, parameterName=Diamond Normal Strength})
  n22 = Normalize(VectorInput=AppendVector(A=AppendVector(A=@n20.R, B=@n20.R), B=Multiply(A=@n20.B, B=@n21)))
  n23 = Add(
   A=RotateAboutAxis(NormalizedRotationAxis=Constant3Vector() {constant=<rgba 0 0 1 0>}, RotationAngle=Add(A=@n6, B=@n9), PivotPoint=Constant(), Position=@n22)
   B=@n22
  )
  n24 = Saturate(None=ConstantBiasScale(None=DotProduct(A=@n23, B=@n4)))
  n25 = ScalarParameter() {defaultValue=2, group=Diamond_Base, parameterName=Internal Power}
  n26 = Multiply(
   A=Multiply(A=Power(Base=@n24, Exp=@n25) {constExponent=2.5}, B=NamedRerouteUsage())
   B=Multiply(A=VectorParameter() {group=Diamond_Base, parameterName=Diamond Internal Color, defaultValue=<rgba 1 1 1 1>}.RGB, B=NamedRerouteUsage())
  )
  n27 = ScalarParameter() {defaultValue=0.5, group=Diamond_Base, parameterName=Reflection Transition Max}
  n28 = ScalarParameter() {group=Diamond_Base, parameterName=Reflection Transition Min}
  n29 = Transform(None=NamedRerouteUsage())
  n30 = Saturate(None=DotProduct(A=@n29, B=CameraVectorWS()))
  n31 = Power(Base=Saturate(None=InverseLinearInterpolate(A=@n27, B=@n28, Value=@n30) {constA=0.9, constB=0.3}))
  n32 = OneMinus(None=@n31)
  n33 = LinearInterpolate(
   A=Multiply(
    A=NamedRerouteUsage()
    B=VectorParameter() {group=Diamond_Base, parameterName=Reflection Color, defaultValue=<rgba 0.3521 0.3521 0.3521 1>}.RGB
   )
   B=VectorParameter() {group=Diamond_Base, parameterName=Reflection Color Outer, defaultValue=<rgba 0.9021 0.9097 1 1>}.RGB
   Alpha=@n31
  )
  n34 = ReflectionVectorWS(CustomWorldNormal=@n29)
  n35 = Multiply(A=NamedRerouteUsage(), B=ScalarParameter() {defaultValue=1, group=Diamond_Base, parameterName=Surface Roughness})
  n36 = Multiply(
   A=Max(
    A=SkyLightEnvMapSample(Direction=@n34, Roughness=@n35)
    B=ScalarParameter() {defaultValue=0.6, group=Diamond_Base, parameterName=Min Reflection Brightness}
   ) {constB=0.5}
   B=NamedRerouteUsage()
  )
  n37 = Multiply(
   A=Power(Base=Saturate(None=@n3), Exp=ScalarParameter() {defaultValue=1, group=Diamond_Base, parameterName=Gap Brightness Power}) {constExponent=1.5}
   B=Max(A=Multiply(A=@n26, B=@n32), B=Multiply(A=@n33, B=@n36))
  )
  n38 = Multiply(A=@n37, B=ScalarParameter() {defaultValue=1, group=Diamond_Base, parameterName=Low End Emissive Multiplier})
  n39 = Divide(
   A=QualitySwitch(Default=@n37, Low=@n38, High=@n37, Medium=@n38, Epic=@n37)
   B=Multiply(A=EyeAdaptation().EyeAdaptation, B=ScalarParameter() {defaultValue=1, group=Diamond_Base, parameterName=Diamond Eye Adaptation Prescale})
  )
  n40 = ScalarParameter() {defaultValue=1, group=Diamond_Base, parameterName=Diamond Eye Adaptation Alpha}
  n41 = SetMaterialAttributes!(
   MaterialAttributes
   BaseColor=LinearInterpolate(A=@n1.BaseColor, Alpha=NamedRerouteUsage()) {constB=0}
   Metallic=LinearInterpolate(A=@n1.Metallic, Alpha=NamedRerouteUsage()) {constB=0}
   Specular=LinearInterpolate(A=@n1.Specular, Alpha=NamedRerouteUsage()) {constA=1, constB=0}
   Roughness=LinearInterpolate(A=@n1.Roughness, Alpha=NamedRerouteUsage()) {constB=0}
   EmissiveColor=LinearInterpolate(A=@n1.EmissiveColor, B=LinearInterpolate(A=@n37, B=@n39, Alpha=@n40), Alpha=NamedRerouteUsage())
   Normal=LinearInterpolate(A=@n1.Normal, B=@n1.Normal, Alpha=NamedRerouteUsage())
  ) {attributeSetTypes [
    #1
    #2
    #3
    #4
    #5
    #6
   ]}
  FunctionOutput(None=@n41)
  Multiply(A=@n29) {constB=-1}
  NamedRerouteDeclaration(None=@n1.Normal) {name=PixelNormal, nodeColor=<rgba 0.6057 0 1 1>}
  NamedRerouteDeclaration(None=StaticSwitchParameter(True=@n1.Specular, False=ScalarParameter() {defaultValue=1, group=Diamond_Base, parameterName=Specularity Multiplier}) {defaultValue=true, group=Diamond_Base, parameterName=Use Specular Texture for Diamonds}) {name=Specularity, nodeColor=<rgba 0.6057 0 1 1>}
  NamedRerouteDeclaration(None=@n1.Roughness) {name=Roughness, nodeColor=<rgba 0.6057 0 1 1>}
  n42 = ScalarParameter() {defaultValue=1, group=Diamond_Base, parameterName=BaseColor_Multiplier}
  NamedRerouteDeclaration(None=Saturate(None=Multiply(A=@n1.BaseColor, B=@n42))) {name=BaseColor, nodeColor=<rgba 0.6057 0 1 1>}
  n43 = Constant4Vector() {constant=<rgba 1 1 1 1>}
  NamedRerouteDeclaration(None=ChannelMaskParameter(None=FunctionInput(Preview=@n43.RGBA) {bUsePreviewValueAsDefault=true, inputName=TechArtMask, sortPriority=1}) {group=Diamond_Base, parameterName=DiamondChannel}.R) {name=TechArtMask, nodeColor=<rgba 0.6641 1 0 1>}
  Normalize(VectorInput=Transform(None=@n14) {transformSourceType=TRANSFORMSOURCE_World, transformType=TRANSFORM_Local})
  Normalize(VectorInput=Transform(None=@n13) {transformSourceType=TRANSFORMSOURCE_World, transformType=TRANSFORM_Local})
  Normalize(VectorInput=Transform(None=@n15) {transformSourceType=TRANSFORMSOURCE_World, transformType=TRANSFORM_Local})
 ]}
