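# M_BerryTart_LumenPlane is an unlit material for the BerryTart set on SharkIsland, named for Lumen planes.
# It multiplies a LumenColor vector parameter, white by default, by a LumenStrength scalar.
# LumenStrength reads from custom primitive data, so intensity can vary per primitive without separate material instances.
# Automatic usage flagging in the editor is turned off.
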Material /Clyde/Environments/Sets/SharkIsland/BerryTart/Materials/M_BerryTart_LumenPlane
graph {
 properties {bAutomaticallySetUsageInEditor=false, shadingModel=MSM_Unlit, usedShadingModels=Unlit}
 nodes [Multiply(
   A=VectorParameter() {parameterName=LumenColor, defaultValue=<rgba 1 1 1 1>}.RGB
   B=ScalarParameter() {bUseCustomPrimitiveData=true, defaultValue=1, parameterName=LumenStrength}
  )]
}
