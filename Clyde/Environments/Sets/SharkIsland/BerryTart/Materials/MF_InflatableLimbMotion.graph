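# Material function that drives noise-based rotational wobble for a masked limb on inflatable props.
# Time, scaled by limb speed and offset, scrolls along a normalized direction through a blurred noise texture, whose red channel blends between negative and positive rotation ranges, scaled by a fade mask.
# It outputs the angle, axis, local pivot, position and transformed normal, each multiplied by the combined vertex and texture-set mask, plus that mask.
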MaterialFunction /Clyde/Environments/Sets/SharkIsland/BerryTart/Materials/MF_InflatableLimbMotion
graph {nodes [
  n1 = FunctionInput() {inputName=Rotation Range Negative 0-1, inputType=FunctionInput_Scalar, sortPriority=5}
  n2 = Add(
   A=Multiply(A=Time(), B=FunctionInput() {inputName=Limb Speed, inputType=FunctionInput_Scalar})
   B=FunctionInput(Preview=Constant()) {bUsePreviewValueAsDefault=true, inputName=Limb Time Offset, inputType=FunctionInput_Scalar, sortPriority=1}
  )
  n3 = Normalize(VectorInput=FunctionInput(Preview=Constant2Vector() {g=0.8, r=0.2}) {bUsePreviewValueAsDefault=true, inputName=Texture Sample Direction, inputType=FunctionInput_Vector2, sortPriority=7})
  n4 = TextureObject() {samplerType=SAMPLERTYPE_Grayscale, texture=&LowResBlurredNoise}
  n5 = LinearInterpolate(
   A=Multiply(A=Saturate(None=@n1)) {constB=-1}
   B=Saturate(None=FunctionInput() {inputName=Rotation Range Positive 0-1, inputType=FunctionInput_Scalar, sortPriority=4})
   Alpha=TextureSample(
    UVs=Multiply(A=@n2, B=@n3)
    Tex=FunctionInput(Preview=@n4) {bUsePreviewValueAsDefault=true, inputName=Motion Texture, inputType=FunctionInput_Texture2D, sortPriority=6}
   ).R
  )
  n6 = FunctionInput(Preview=Constant() {r=1}) {bUsePreviewValueAsDefault=true, inputName=Rotation Fade Mask, inputType=FunctionInput_Scalar, sortPriority=8}
  n7 = Multiply(
   A=FunctionInput(Preview=Constant() {r=1}) {bUsePreviewValueAsDefault=true, inputName=Vertex Mask, inputType=FunctionInput_Scalar, sortPriority=9}
   B=FunctionInput(Preview=Constant() {r=1}) {bUsePreviewValueAsDefault=true, inputName=Texture Set Mask, inputType=FunctionInput_Scalar, sortPriority=10}
  )
  FunctionOutput(None=Multiply(A=Multiply(A=@n5, B=@n6), B=@n7)) {outputName=RotationAngle}
  n8 = FunctionInput() {inputName=Rotation Axis Local, sortPriority=2, previewValue=<0 1 0 1>}
  FunctionOutput(None=Multiply(A=Normalize(VectorInput=@n8), B=@n7)) {outputName=NormalizedRotationAxis, sortPriority=1}
  FunctionOutput(None=Multiply(A=FunctionInput() {inputName=Rotation Local Pivot, sortPriority=3}, B=@n7)) {outputName=PivotPoint, sortPriority=2}
  FunctionOutput(None=Multiply(A=LocalPosition().XYZ, B=@n7)) {outputName=Position, sortPriority=3}
  n9 = FunctionInput(Preview=Constant3Vector() {constant=<rgba 0 0 1 0>}) {bUsePreviewValueAsDefault=true, inputName=Normal Texture Sample, sortPriority=11}
  FunctionOutput(None=Multiply(A=Transform(None=@n9) {transformType=TRANSFORM_Local}, B=@n7)) {outputName=Normal, sortPriority=4}
  FunctionOutput(None=@n7) {outputName=Limb Mask, sortPriority=5}
 ]}
