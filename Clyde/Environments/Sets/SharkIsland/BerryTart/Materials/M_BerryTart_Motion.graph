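# Material for the BerryTart inflatable on Shark Island, animating its arms, legs and head with vertex motion.
# Four texture sets blend by thresholds on the second UV channel's U coordinate.
# Five MF_InflatableLimbMotion calls, using per-limb pivots, axes, noise and masks from vertex color and local-position fades, sum into one RotateVertexAboutAxis.
# Its outputs provide World Position Offset and a tangent-space normal for lerping over the blended normal.
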Material /Clyde/Environments/Sets/SharkIsland/BerryTart/Materials/M_BerryTart_Motion
graph {
 properties {bAutomaticallySetUsageInEditor=false, bUseMaterialAttributes=true, usedShadingModels=Default Lit}
 nodes [
  n1 = If(A=Constant() {r=1}, B=ComponentMask(None=TextureCoordinate() {coordinateIndex=1}) {r=true}, "A > B"=Constant(), "A < B"=Constant() {r=1}) {desc=0}
  NamedRerouteDeclaration(None=OneMinus(None=@n1)) {name=HeadMaterialID, nodeColor=<rgba 1 0.5391 0 1>}
  n2 = VertexColor()
  NamedRerouteDeclaration(None=Multiply(A=@n2.G, B=@n2.G)) {name=VertexMaskGreen1, nodeColor=<rgba 1 0 0.0234 1>}
  NamedRerouteDeclaration(None=Multiply(A=@n2.G, B=OneMinus(None=@n2.A))) {name=VertexMaskGreen0, nodeColor=<rgba 1 0 0.0234 1>}
  n3 = MaterialFunctionCall!(
   Limb Speed=ScalarParameter() {defaultValue=0.22, parameterName=ArmSpeed}
   Limb Time Offset=ScalarParameter() {parameterName=RightArmTimeOffset}
   Rotation Axis Local=VectorParameter() {parameterName=Right Arm Rotation Axis, defaultValue=<rgba 0.5 1 0 1>}.RGB
   Rotation Local Pivot=VectorParameter() {parameterName=Right Arm Pivot, defaultValue=<rgba -500 -300 1300 0>}.RGB
   Rotation Range Positive 0-1=ScalarParameter() {defaultValue=0.015, parameterName=ArmOffsetStrength}
   Rotation Range Negative 0-1=Constant()
   Motion Texture=NamedRerouteUsage()
   Texture Sample Direction
   Rotation Fade Mask=NamedRerouteUsage()
   Vertex Mask=NamedRerouteUsage()
   Texture Set Mask=NamedRerouteUsage()
   Normal Texture Sample=NamedRerouteUsage()
  ) {materialFunction=&/MF_InflatableLimbMotion}
  n4 = MaterialFunctionCall(
   Limb Speed=ScalarParameter() {defaultValue=0.22, parameterName=ArmSpeed}
   Limb Time Offset=ScalarParameter() {parameterName=LeftArmTimeOffset}
   Rotation Axis Local=VectorParameter() {parameterName=Left Arm Rotation Axis, defaultValue=<rgba -0.5 1 0 1>}.RGB
   Rotation Local Pivot=VectorParameter() {parameterName=Left Arm Pivot, defaultValue=<rgba -500 -300 1300 0>}.RGB
   Rotation Range Positive 0-1=ScalarParameter() {defaultValue=0.015, parameterName=ArmOffsetStrength}
   Rotation Range Negative 0-1=Constant()
   Motion Texture=NamedRerouteUsage()
   Rotation Fade Mask=NamedRerouteUsage()
   Vertex Mask=NamedRerouteUsage()
   Texture Set Mask=NamedRerouteUsage()
   Normal Texture Sample=NamedRerouteUsage()
  ) {materialFunction=&MF_InflatableLimbMotion}
  n5 = Add(A=@n3.Limb Mask, B=@n4.Limb Mask)
  n6 = MaterialFunctionCall(
   Limb Speed=ScalarParameter() {defaultValue=0.2, parameterName=LegSpeed}
   Limb Time Offset=ScalarParameter() {parameterName=RightLegTimeOffset}
   Rotation Axis Local=VectorParameter() {parameterName=Right Leg Rotation Axis, defaultValue=<rgba 1 0 -0.5 1>}.RGB
   Rotation Local Pivot=VectorParameter() {parameterName=Right Leg Pivot, defaultValue=<rgba 0 350 300 0>}.RGB
   Rotation Range Positive 0-1=ScalarParameter() {defaultValue=0.015, parameterName=LegOffsetStrength}
   Rotation Range Negative 0-1=Constant()
   Motion Texture=NamedRerouteUsage()
   Rotation Fade Mask=NamedRerouteUsage()
   Vertex Mask=NamedRerouteUsage()
   Texture Set Mask=NamedRerouteUsage()
   Normal Texture Sample=NamedRerouteUsage()
  ) {materialFunction=&MF_InflatableLimbMotion}
  n7 = MaterialFunctionCall(
   Limb Speed=ScalarParameter() {defaultValue=0.2, parameterName=LegSpeed}
   Limb Time Offset=ScalarParameter() {defaultValue=0.05, parameterName=LeftLegTimeOffset}
   Rotation Axis Local=VectorParameter() {parameterName=Left Leg Rotation Axis, defaultValue=<rgba 1 0 -0.5 1>}.RGB
   Rotation Local Pivot=VectorParameter() {parameterName=Left Leg Pivot, defaultValue=<rgba 0 350 300 0>}.RGB
   Rotation Range Positive 0-1=ScalarParameter() {defaultValue=0.015, parameterName=LegOffsetStrength}
   Rotation Range Negative 0-1=Constant()
   Motion Texture=NamedRerouteUsage()
   Rotation Fade Mask=NamedRerouteUsage()
   Vertex Mask=NamedRerouteUsage()
   Texture Set Mask=NamedRerouteUsage()
   Normal Texture Sample=NamedRerouteUsage()
  ) {materialFunction=&MF_InflatableLimbMotion}
  n8 = Add(A=@n6.Limb Mask, B=@n7.Limb Mask)
  n9 = ScalarParameter() {defaultValue=0.03, parameterName=HeadOffsetStrength}
  n10 = MaterialFunctionCall(
   Limb Speed=ScalarParameter() {defaultValue=0.1, parameterName=HeadSpeed}
   Rotation Axis Local=VectorParameter() {parameterName=Head Rotation Axis, defaultValue=<rgba 0 1 0 1>}.RGB
   Rotation Local Pivot=VectorParameter() {parameterName=Head Pivot, defaultValue=<rgba -100 -200 1300 1>}.RGB
   Rotation Range Positive 0-1=@n9
   Rotation Range Negative 0-1=@n9
   Motion Texture=NamedRerouteUsage()
   Vertex Mask=VertexColor().R
   Texture Set Mask=NamedRerouteUsage()
   Normal Texture Sample=NamedRerouteUsage()
  ) {materialFunction=&MF_InflatableLimbMotion}
  NamedRerouteDeclaration(None=Add(A=Add(A=@n5, B=@n8), B=@n10.Limb Mask)) {name=FinalLimbMask, nodeColor=<rgba 0.8516 0 1 1>}
  n11 = Add(A=@n3.RotationAngle, B=@n4.RotationAngle)
  n12 = Add(A=@n6.RotationAngle, B=@n7.RotationAngle)
  n13 = Add(A=@n3.NormalizedRotationAxis, B=@n4.NormalizedRotationAxis)
  n14 = Add(A=@n6.NormalizedRotationAxis, B=@n7.NormalizedRotationAxis)
  n15 = Add(A=@n3.PivotPoint, B=@n4.PivotPoint)
  n16 = Add(A=@n6.PivotPoint, B=@n7.PivotPoint)
  n17 = Add(A=@n3.Position, B=@n4.Position)
  n18 = Add(A=@n6.Position, B=@n7.Position)
  n19 = Add(A=@n3.Normal, B=@n4.Normal)
  n20 = Add(A=@n6.Normal, B=@n7.Normal)
  n21 = MaterialFunctionCall!(
   RotationAngle=Add(A=Add(A=@n11, B=@n12), B=@n10.RotationAngle)
   NormalizedRotationAxis=Add(A=Add(A=@n13, B=@n14), B=@n10.NormalizedRotationAxis)
   PivotPoint=Add(A=Add(A=@n15, B=@n16), B=@n10.PivotPoint)
   Position=Add(A=Add(A=@n17, B=@n18), B=@n10.Position)
   Normal=Add(A=Add(A=@n19, B=@n20), B=@n10.Normal)
   TangentU
   TangentV
  ) {materialFunction=/Game/Characters/Player/Male/Medium/Bodies/M_MED_Inferno/FX/Mat/RotateVertexAboutAxis}
  NamedRerouteDeclaration(None=Transform(None=@n21.WPO) {transformSourceType=TRANSFORMSOURCE_Local}) {name=FinalWPO, nodeColor=<rgba 0.8516 0 1 1>}
  NamedRerouteDeclaration(None=Saturate(None=Transform(None=@n21.Normal) {transformSourceType=TRANSFORMSOURCE_Local, transformType=TRANSFORM_Tangent})) {name=FinalNormal, nodeColor=<rgba 0.8516 0 1 1>}
  n22 = If(A=Constant() {r=3}, B=ComponentMask(None=TextureCoordinate() {coordinateIndex=1}) {r=true}, "A > B"=Constant(), "A < B"=Constant() {r=1}) {desc=0}
  NamedRerouteDeclaration(None=@n22) {name=LimbsMaterialID, nodeColor=<rgba 1 0.5391 0 1>}
  NamedRerouteDeclaration(None=TextureObject() {samplerType=SAMPLERTYPE_Grayscale, texture=&LowResBlurredNoise}) {name=MotionTexture, nodeColor=<rgba 0 0.7422 1 1>}
  n23 = VectorParameter() {parameterName=LeftLegStartFade, defaultValue=<rgba 0 535 0 0>}
  n24 = Subtract(A=VectorParameter() {parameterName=LeftLegEndFade, defaultValue=<rgba 0 650 0 0>}.RGB, B=@n23.RGB)
  n25 = DotProduct(A=Subtract(A=LocalPosition().XYZ, B=@n23.RGB), B=Normalize(VectorInput=@n24))
  n26 = Length(None=@n24) {is=class MaterialExpressionLength}
  NamedRerouteDeclaration(None=Saturate(None=Divide(A=@n25, B=@n26) {constB=100})) {name=LeftLegFade, nodeColor=<rgba 1 0.9375 0 1>}
  n27 = VectorParameter() {parameterName=RightLegStartFade, defaultValue=<rgba 0 535 0 0>}
  n28 = Subtract(A=VectorParameter() {parameterName=RightLegEndFade, defaultValue=<rgba 0 650 0 0>}.RGB, B=@n27.RGB)
  n29 = DotProduct(A=Subtract(A=LocalPosition().XYZ, B=@n27.RGB), B=Normalize(VectorInput=@n28))
  n30 = Length(None=@n28)
  NamedRerouteDeclaration(None=Saturate(None=Divide(A=@n29, B=@n30) {constB=100})) {name=RightLegFade, nodeColor=<rgba 1 0.9375 0 1>}
  n31 = VectorParameter() {parameterName=ArmStartFade, defaultValue=<rgba 0 0 900 0>}
  n32 = Subtract(A=VectorParameter() {parameterName=ArmEndFade, defaultValue=<rgba 0 0 200 0>}.RGB, B=@n31.RGB)
  n33 = DotProduct(A=Subtract(A=LocalPosition().XYZ, B=@n31.RGB), B=Normalize(VectorInput=@n32))
  n34 = Length(None=@n32)
  NamedRerouteDeclaration(None=Saturate(None=Divide(A=@n33, B=@n34) {constB=100})) {name=ArmFade, nodeColor=<rgba 1 0.9375 0 1>}
  n35 = VertexColor()
  NamedRerouteDeclaration(None=Multiply(A=@n35.R, B=@n35.R)) {name=VertexMaskRed1, nodeColor=<rgba 1 0 0.0234 1>}
  NamedRerouteDeclaration(None=Multiply(A=@n35.R, B=OneMinus(None=@n35.A))) {name=VertexMaskRed0, nodeColor=<rgba 1 0 0.0234 1>}
  n36 = TextureSample() {samplerType=SAMPLERTYPE_Masks, texture=/Clyde/Environments/Sets/SharkIsland/BerryTart/Textures/T_BerryTart_Inflatable_A_A_S}
  n37 = SetMaterialAttributes!(
   MaterialAttributes
   Base Color=TextureSample() {texture=/Clyde/Environments/Sets/SharkIsland/BerryTart/Textures/T_BerryTart_Inflatable_A_A_D}.RGB
   Metallic=@n36.G
   Specular=@n36.G
   Roughness=@n36.G
   Normal=NamedRerouteDeclaration(None=TextureSample() {samplerType=SAMPLERTYPE_Normal, texture=/Clyde/Environments/Sets/SharkIsland/BerryTart/Textures/T_BerryTart_Inflatable_A_A_N}.RGB) {name=HeadNormal, nodeColor=<rgba 0 0 1 1>}
  ) {attributeSetTypes [
    #1
    #2
    #3
    #4
    #5
   ]}
  n38 = TextureSample() {samplerType=SAMPLERTYPE_Masks, texture=/Clyde/Environments/Sets/SharkIsland/BerryTart/Textures/T_BerryTart_Inflatable_A_B_S}
  n39 = SetMaterialAttributes(
   Base Color=TextureSample() {texture=/Clyde/Environments/Sets/SharkIsland/BerryTart/Textures/T_BerryTart_Inflatable_A_B_D}.RGB
   Metallic=@n38.G
   Specular=@n38.G
   Roughness=@n38.G
   Normal=TextureSample() {samplerType=SAMPLERTYPE_Normal, texture=/Clyde/Environments/Sets/SharkIsland/BerryTart/Textures/T_BerryTart_Inflatable_A_B_N}.RGB
  ) {attributeSetTypes [
    #1
    #2
    #3
    #4
    #5
   ]}
  n40 = TextureSample() {samplerType=SAMPLERTYPE_Masks, texture=/Clyde/Environments/Sets/SharkIsland/BerryTart/Textures/T_BerryTart_Inflatable_A_C_S}
  n41 = SetMaterialAttributes(
   Base Color=TextureSample() {texture=/Clyde/Environments/Sets/SharkIsland/BerryTart/Textures/T_BerryTart_Inflatable_A_C_D}.RGB
   Metallic=@n40.G
   Specular=@n40.G
   Roughness=@n40.G
   Normal=TextureSample() {samplerType=SAMPLERTYPE_Normal, texture=/Clyde/Environments/Sets/SharkIsland/BerryTart/Textures/T_BerryTart_Inflatable_A_C_N}.RGB
  ) {attributeSetTypes [
    #1
    #2
    #3
    #4
    #5
   ]}
  n42 = Constant() {r=2}
  n43 = ComponentMask(None=TextureCoordinate() {coordinateIndex=1}) {r=true}
  n44 = Constant()
  n45 = Constant() {r=1}
  n46 = TextureSample() {samplerType=SAMPLERTYPE_Masks, texture=/Clyde/Environments/Sets/SharkIsland/BerryTart/Textures/T_BerryTart_Inflatable_A_D_S}
  n47 = SetMaterialAttributes(
   Base Color=TextureSample() {texture=/Clyde/Environments/Sets/SharkIsland/BerryTart/Textures/T_BerryTart_Inflatable_A_D_D}.RGB
   Metallic=@n46.G
   Specular=@n46.G
   Roughness=@n46.G
   Normal=NamedRerouteDeclaration(None=TextureSample() {samplerType=SAMPLERTYPE_Normal, texture=/Clyde/Environments/Sets/SharkIsland/BerryTart/Textures/T_BerryTart_Inflatable_A_D_N}.RGB) {name=LimbsNormal, nodeColor=<rgba 0 0 1 1>}
  ) {attributeSetTypes [
    #1
    #2
    #3
    #4
    #5
   ]}
  n48 = BlendMaterialAttributes(
   A=BlendMaterialAttributes(A=BlendMaterialAttributes(A=@n37, B=@n39, Alpha=@n1), B=@n41, Alpha=If(A=@n42, B=@n43, "A > B"=@n44, "A < B"=@n45) {desc=0})
   B=@n47
   Alpha=@n22
  )
  SetMaterialAttributes!(
   MaterialAttributes=@n48
   Normal=LinearInterpolate(
    A=GetMaterialAttributes(None=@n48) {attributeGetTypes [#5]}.Normal
    B=NamedRerouteUsage()
    Alpha=NamedRerouteUsage()
   )
   World Position Offset=NamedRerouteUsage()
  ) {attributeSetTypes [#5, #6]}
 ]
}
