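# Fortnite four-layer base master material variant for display screens in the SharkIsland BerryTart set.
# Layered attributes pass through shared detail, weathering and effect functions, with time-of-day emissive scaling and optional eye-adaptation compensation.
# Time-driven sine sums glitch layer 2 UVs into pixelated steps and trigger noise-based dead pixels, alongside scrolling scanlines and warp.
# View-angle darkening dims the screen, bypassed for ray tracing, before shared WPO, RVT blending and mobile simplification.
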Material /Clyde/Environments/Sets/SharkIsland/BerryTart/Materials/M_BerryTart_Screen
graph {
 properties {
  bAutomaticallySetUsageInEditor=false
  bTangentSpaceNormal=false
  bUseMaterialAttributes=true
  bUsedWithInstancedStaticMeshes=true
  bUsedWithNanite=true
  bUsedWithSplineMeshes=true
  ditheredLODTransition=true
  pixelDepthOffsetMode=PDOM_Legacy
  shadingModel=MSM_FromMaterialExpression
  usedShadingModels=Default Lit | Subsurface | Clear Coat | Two Sided Foliage
 }
 nodes [
  n1 = NamedRerouteUsage()
  n2 = NamedRerouteUsage()
  n3 = MakeMaterialAttributes!(
   BaseColor
   Metallic
   Specular
   Roughness
   Anisotropy
   EmissiveColor
   Opacity
   OpacityMask
   Normal
   Tangent
   WorldPositionOffset
   SubsurfaceColor
   ClearCoat
   ClearCoatRoughness
   AmbientOcclusion
   Refraction
   CustomizedUVs_0
   CustomizedUVs_1
   CustomizedUVs_2
   CustomizedUVs_3
   CustomizedUVs_4
   CustomizedUVs_5
   CustomizedUVs_6
   CustomizedUVs_7
   PixelDepthOffset
   ShadingModel
   Displacement
  )
  n4 = MaterialFunctionCall!(SFX V3=NamedRerouteUsage(), 1=NamedRerouteUsage(), 2=NamedRerouteUsage(), 3=@n3, 4=@n3) {materialFunction=/Game/Global/Base4Layers_2022/MaterialFunctions/MF_Layer_4to1}
  MaterialExpressionReroute_12 = Reroute(None=@n4.Result)
  n5 = GetMaterialAttributes(None=@MaterialExpressionReroute_12.Input) {attributeGetTypes [#1, #2]}
  n6 = StaticSwitchParameter(
   True=StaticSwitchParameter(True=TextureCoordinate() {coordinateIndex=2}, False=TextureCoordinate() {coordinateIndex=1}) {group=B_SSS_EdgeDepth, parameterName=SSS Depth UV2, sortPriority=-30}
   False=TextureCoordinate()
  ) {desc="Default is UV0 ", group=B_SSS_EdgeDepth, parameterName=SSS Depth UV1, sortPriority=-30}
  n7 = AppendVector(
   A=ScalarParameter() {defaultValue=1, group=B_SSS_EdgeDepth, parameterName=SSSDepth X, sortPriority=5}
   B=ScalarParameter() {defaultValue=1, group=B_SSS_EdgeDepth, parameterName=SSSDepth Y, sortPriority=5}
  )
  n8 = Add(
   A=ComponentMask(None=MaterialFunctionCall!(UVs=@n6, Tiling=@n7, Randomize Rotation) {materialFunction=/Engine/Functions/Engine_MaterialFunctions01/Cubemaps/InteriorCubemap}.UVW) {r=true}
   B=Constant() {r=0.98}
  )
  n9 = ScalarParameter() {defaultValue=3, group=B_SSS_EdgeDepth, parameterName=SSS_Depth Edge Blur, sortPriority=3}
  n10 = LinearInterpolate(A=ScalarParameter() {group=B_SSS_EdgeDepth, parameterName=SSS Depth Min, sortPriority=4}, Alpha=Saturate(None=Divide(A=@n8, B=@n9)))
  n11 = ScalarParameter() {group=B_SSS_EdgeDepth, parameterName=SSS Fresnel Min, sortPriority=1}
  n12 = Saturate(None=Fresnel(
    ExponentIn=ScalarParameter() {defaultValue=3, group=B_SSS_EdgeDepth, parameterName=SSS Fresnel Exponent, sortPriority=2}
    Normal=Transform(None=@n5.Normal)
   ) {baseReflectFraction=0, exponent=3})
  n13 = Saturate(None=Multiply(A=Multiply(A=@n5.Subsurface Color, B=@n10), B=LinearInterpolate(B=@n11, Alpha=@n12) {constA=1}))
  n14 = StaticSwitchParameter(
   True=QualitySwitch(
    Default=@MaterialExpressionReroute_12.Input
    Epic=SetMaterialAttributes!(MaterialAttributes=@n5.MaterialAttributes, Subsurface Color=@n13) {attributeSetTypes [#1]}
   )
   False=@MaterialExpressionReroute_12.Input
  ) {group=B_SSS_EdgeDepth, parameterName=SSS_EdgeDepth}
  MaterialExpressionReroute_14 = Reroute(None=@n14)
  n15 = NamedRerouteUsage()
  n16 = MaterialFunctionCall!(
   SFX V3=@n1
   Mat In=MaterialFunctionCall!(
    Material In=MaterialFunctionCall!(SFX=@n2, Material In=@MaterialExpressionReroute_14.Output, SFX Textures RGB=@n15) {materialFunction=/Game/Global/Base4Layers_2022/MaterialFunctions/MF_EmissiveControls_2}.Result
    SFX Texture=NamedRerouteUsage()
    SFX RGB=NamedRerouteUsage()
   ) {materialFunction=/Game/Global/Base4Layers_2022/MaterialFunctions/MF_StickersPrint}.Result
  ) {materialFunction=/Game/Global/Base4Layers_2022/MaterialFunctions/MF_DetailsWS}
  n17 = NamedRerouteUsage()
  n18 = MaterialFunctionCall!(
   In=MaterialFunctionCall!(
    AO=@n16.AO Mask
    In=MaterialFunctionCall!(Mat In=@n16.Result) {materialFunction=/Game/Global/Base4Layers_2022/MaterialFunctions/MF_SnowWS}.Result
   ) {materialFunction=/Game/Global/Base4Layers_2022/MaterialFunctions/MF_Leaks}.Result
   SFXMask=@n17
  ) {materialFunction=/Game/Global/Base4Layers_2022/MaterialFunctions/MF_4Layer_ThinFilm}
  n19 = GetMaterialAttributes(None=MaterialFunctionCall!(
    Layer Debug
    Debug Layer Color
    Mat=MaterialFunctionCall!(In=@n18.Result, SFXMask=@n17) {materialFunction=/Game/Global/Base4Layers_2022/MaterialFunctions/MF_DiffuseHueShift}.Result
   ) {materialFunction=/Game/Global/Base4Layers_2022/MaterialFunctions/MF_DebugViews}.POI Color) {attributeGetTypes [#3]}
  n20 = Multiply(
   A=@n19.Emissive Color
   B=LinearInterpolate(
    A=ScalarParameter() {defaultValue=0.1, parameterName=EmissiveMultiMin}
    B=ScalarParameter() {defaultValue=1, parameterName=EmissiveMultiMax}
    Alpha=Desaturation(None=@n19.Emissive Color)
   )
  )
  n21 = ScalarParameter() {defaultValue=1, parameterName=ScreenEmissiveMasterDay}
  n22 = ScalarParameter() {defaultValue=1, parameterName=ScreenEmissiveMasterNight}
  n23 = LinearInterpolate(A=@n21, B=@n22)
  n24 = Multiply(
   A=@n20
   B=MaterialFunctionCall(Value_Morning=@n23, Value_Day=@n21, Value_Evening=@n23, Value_Night=@n22) {materialFunction=&Scalar_Time_of_Day}.Result
  )
  n25 = GetMaterialAttributes(None=SetMaterialAttributes!(MaterialAttributes=@n19.MaterialAttributes, Emissive Color=RayTracingQualitySwitch(Normal=@n24, RayTraced=@n20)) {attributeSetTypes [#3]}) {attributeGetTypes [#4, #3]}
  n26 = ScalarParameter() {defaultValue=1, parameterName=SaturationBaseColorBoost}
  n27 = NamedRerouteUsage()
  n28 = Add(A=Saturate(None=Multiply(A=@n25.Base Color, B=@n26)), B=@n27)
  n29 = CameraVectorWS()
  n30 = VertexNormalWS()
  n31 = MaterialFunctionCall(
   Input=Abs(None=DotProduct(A=@n29, B=@n30))
   Input Low=ScalarParameter() {defaultValue=0.05, parameterName=MinDarkening}
   Input High=Constant() {r=1}
  ) {materialFunction=&RemapValue_0-1}
  n32 = ScalarParameter() {defaultValue=1.5, parameterName=ScreenDarkeningPower}
  n33 = Saturate(None=Multiply(A=Power(Base=@n31.Result, Exp=@n32) {constExponent=1.5}) {constB=2})
  n34 = Add(A=@n25.Emissive Color, B=@n27)
  n35 = SetMaterialAttributes!(
   MaterialAttributes=@n25.MaterialAttributes
   Base Color=RayTracingQualitySwitch(Normal=Multiply(A=@n28, B=@n33), RayTraced=@n28)
   Emissive Color=RayTracingQualitySwitch(Normal=Multiply(A=@n34, B=@n33), RayTraced=@n34)
  ) {attributeSetTypes [#4, #3]}
  n36 = GetMaterialAttributes(None=@n35) {attributeGetTypes [#4, #3]}
  n37 = VectorParameter() {parameterName=DeadPixels.Color, defaultValue=<rgba 0 0 0 1>}
  n38 = SetMaterialAttributes(
   MaterialAttributes=@n36.MaterialAttributes
   Base Color=Multiply(A=@n36.Base Color, B=@n37.RGB)
   Emissive Color=Multiply(A=@n36.Emissive Color, B=@n37.RGB)
  ) {attributeSetTypes [#4, #3]}
  n39 = Multiply(A=ComponentMask(None=NamedRerouteUsage()) {r=true}, B=ScalarParameter() {defaultValue=1, parameterName=DeadPixels.Tiling})
  n40 = ScalarParameter() {parameterName=DeadPixels.Offset}
  n41 = Frac(None=ActorPositionWS())
  n42 = Multiply(A=Time(), B=ScalarParameter() {defaultValue=0.25, parameterName=DeadPixels.Speed})
  n43 = Add(A=ConstantBiasScale(None=Sine(None=@n42)), B=ConstantBiasScale(None=Sine(None=@n42) {period=1.33}))
  n44 = Multiply(A=ConstantBiasScale(None=Sine(None=@n42) {period=1.71})) {constB=0.35}
  n45 = Add(A=@n42, B=Multiply(A=Add(A=@n43, B=@n44)) {constB=0.2})
  n46 = ScalarParameter() {defaultValue=8, parameterName=DeadPixels.Steps}
  n47 = ScalarParameter() {defaultValue=8, parameterName=DeadPixels.Steps}
  n48 = Frac(None=Multiply(A=Multiply(A=@n45, B=@n46), B=Divide(B=@n47)))
  n49 = Divide(A=Floor(None=Multiply(A=@n48, B=@n47) {constB=4}), B=@n47) {constB=4}
  n50 = Multiply(
   A=Add(
    A=Add(A=@n39, B=@n40) {constB=0.1}
    B=MaterialFunctionCall(f3=@n41) {materialFunction=&AddComponents}.f3
   )
   B=LinearInterpolate(B=LinearInterpolate(Alpha=@n49) {constA=0.7, constB=1.1}, Alpha=ScalarParameter() {defaultValue=1, parameterName=DeadPixels.Animate}) {constA=1}
  )
  n51 = TextureSample!(UVs=@n50, Tex, Level=Constant() {r=4}, Apply View MipBias) {mipValueMode=TMVM_MipLevel, texture=&AtmosphericCloudNoise01}
  n52 = Multiply(A=@n51.R, B=ScalarParameter() {defaultValue=9, parameterName=DeadPixels.MultiA}) {constB=10}
  n53 = Round(None=Add(A=Sine(None=@n52), B=ScalarParameter() {parameterName=DeadPixels.MultiA.Offset}))
  n54 = OneMinus(None=@n51.R)
  n55 = ScalarParameter() {defaultValue=9, parameterName=DeadPixels.MultiB}
  n56 = Round(None=Sine(None=Multiply(A=@n54, B=@n55) {constB=9}))
  n57 = Time()
  n58 = Add(A=Add(A=Sine(None=@n57), B=Sine(None=@n57) {period=0.45}), B=Sine(None=@n57) {period=1.63})
  n59 = LinearInterpolate(B=Saturate(None=Multiply(A=@n53, B=@n56)), Alpha=If(A=Saturate(None=@n58), "A > B"=Constant() {r=1}, "A < B"=Constant()) {constB=0.95})
  n60 = GetMaterialAttributes(None=BlendMaterialAttributes(A=@n25.MaterialAttributes, B=BlendMaterialAttributes(A=@n35, B=@n38, Alpha=@n59), Alpha=NamedRerouteUsage())) {attributeGetTypes [#3]}
  n61 = EyeAdaptation()
  n62 = SetMaterialAttributes!(
   MaterialAttributes=@n60.MaterialAttributes
   Emissive Color=LinearInterpolate(
    A=@n60.Emissive Color
    B=Divide(A=@n60.Emissive Color, B=@n61.EyeAdaptation)
    Alpha=ScalarParameter() {parameterName=ScreenEyeAdaptationAlpha}
   )
  ) {attributeSetTypes [#3]}
  MaterialExpressionReroute_70 = Reroute(None=@n62)
  n63 = ScalarParameter() {
   desc=Documentation link
   group="0 (Expand)  Documentation (RMB - Copy display name) and Debug view"
   parameterName=https://confluence.it.epicgames.com/display/FORT/Fortnite+Base+Material+2022
  }
  n64 = SetMaterialAttributes!(MaterialAttributes=@MaterialExpressionReroute_70.MaterialAttributes, World Position Offset=Multiply(A=@n63)) {attributeSetTypes [#5]}
  MaterialExpressionReroute_71 = Reroute(None=@n64)
  MaterialExpressionReroute_69 = Reroute(None=@MaterialExpressionReroute_71.Output)
  n65 = MaterialFunctionCall!(Material Attributes=MaterialFunctionCall!(In=MaterialFunctionCall(Material_In=@MaterialExpressionReroute_69.MaterialAttributes) {materialFunction=&MF_FN_DestructibleDamage}.Material_Out) {materialFunction=/Game/Athena/WinterArt/Materials/MF_WPO_WindNew}.Result) {materialFunction=/Game/Creative/Sets/MilitaryBase/Props/MF_FanSpin}
  n66 = MaterialFunctionCall!(In=MaterialFunctionCall!(
    In=MaterialFunctionCall!(In=@n65.Result, UseSphericalMappingOverride) {materialFunction=/Game/Packages/Fortress_SharedMaterials/Base_Material_Functions/MF_CurieFallbackEffect}.Result
    Custom Mask
    Custom Rotation Axis
   ) {materialFunction=/Game/Athena/WinterArt/Materials/MF_WindWPO_v2}.Result) {materialFunction=/Game/Global/Base4Layers_2022/MaterialFunctions/MF_SnowAndIceArchitecture}
  MaterialExpressionReroute_68 = Reroute(None=@n66.Result)
  Composite!(None=@MaterialExpressionReroute_68.Result) {desc=Transform to World in case we will use RVT after this point, subgraphName=Tangent to World}
  Composite!(MaterialAttributes=@n62) {subgraphName=Documentation Link}
  Composite(MaterialAttributes=@MaterialExpressionReroute_71.Output) {desc=Game Shared WPO and Fire Effect, subgraphName=WPO And Shared Functions}
  Composite!(None=@n4.Result)
  MaterialExpressionReroute_1 = Reroute(None=@n73.RVT H)
  MaterialExpressionReroute_3 = Reroute(None=@MaterialExpressionReroute_68.Result)
  n67 = GetMaterialAttributes(None=@MaterialExpressionReroute_3.Input) {attributeGetTypes [#2]}
  n68 = Transform(None=@n67.Normal)
  n69 = LinearInterpolate(A=Multiply(A=@n68) {constB=-1}, B=@n68, Alpha=Saturate(None=TwoSidedSign())) {desc="Fixes normals\r\non backfaces"}
  n70 = SetMaterialAttributes!(
   MaterialAttributes=@n67.MaterialAttributes
   Normal=StaticSwitchParameter(True=MaterialProxyReplace(Realtime=@n69, MaterialProxy=@n67.Normal), False=@n69) {parameterName=Force TangentSpaceNormalsFor HLOD ProxyReplace}
  ) {attributeSetTypes [#2]}
  MaterialExpressionReroute_2 = Reroute(None=@n70)
  n71 = MaterialFunctionCall!(In=MaterialFunctionCall!(In=@MaterialExpressionReroute_2.Output) {materialFunction=/Game/ArtTools/Tech_Art/DirectFX/Assets/MaterialFunctions/MF_DirectFX}.Result) {materialFunction=/Game/ArtTools/Tech_Art/Veneer/MaterialFunctions/MF_Veneer}
  MaterialExpressionReroute_8 = Reroute(None=@n71.Result)
  n72 = MaterialFunctionCall!(AO, RVT_Height=@MaterialExpressionReroute_1."RVT_Height (S)", For Buildings=@MaterialExpressionReroute_8."For Moss (MA)", For Moss) {materialFunction=/Game/Environments/Asteria/Rocks/Materials/MaterialFunctions/MF_JungleBiome}
  MaterialExpressionReroute_0 = Reroute(None=@n72.Buildings Jungl)
  n73 = MaterialFunctionCall!(
   In=@MaterialExpressionReroute_0.Result
   "RVT Height Map (Roughness)"=GetMaterialAttributes(None=@n71.Result) {attributeGetTypes [#6]}.Roughness
  ) {materialFunction=/Game/Global/Base4Layers_2022/MaterialFunctions/MF_RVTBlending}
  Composite!(InputPin=@n73.Result) {subgraphName=Mobile Opt}
  Composite!() {subgraphName=L2 Blending Logic help}
  Composite!("For Moss (MA)"=@n71.Result, "RVT_Height (S)"=@n73.RVT H) {subgraphName="Seasonal "}
  Desaturation()
  n74 = Time()
  n75 = Add(A=Sine(None=@n74), B=Sine(None=@n74) {period=0.45})
  n76 = Sine(None=@n74) {period=1.63}
  n77 = Multiply(A=ScalarParameter() {defaultValue=100, parameterName=ScreenResolution}) {constB=2}
  n78 = Multiply(A=@n50, B=@n77) {constB=100}
  If(A=Saturate(None=Add(A=@n75, B=@n76)), "A > B"=Divide(A=Ceil(None=@n78), B=@n77) {constB=100}, "A < B"=@n50) {constB=0.95}
  MaterialExpressionReroute_7 = Reroute(None=@n73.Result)
  n79 = GetMaterialAttributes(None=@MaterialExpressionReroute_7.InputPin)
  n80 = Constant() {r=0.5}
  n81 = Constant() {r=0.8}
  n82 = ShadingPathSwitch(
   Default=@MaterialExpressionReroute_7.InputPin
   Mobile=QualitySwitch(
    Default=@MaterialExpressionReroute_7.InputPin
    Low=SetMaterialAttributes!(MaterialAttributes=@n79.MaterialAttributes, Specular=@n80, Roughness=@n81) {attributeSetTypes [#7, #6]}
    High=@MaterialExpressionReroute_7.InputPin
    Medium=@MaterialExpressionReroute_7.InputPin
    Epic=@MaterialExpressionReroute_7.InputPin
   )
  )
  MaterialExpressionReroute_6 = Reroute(None=@n82)
  MaterialFunctionCall(MA In=@MaterialExpressionReroute_6.Output) {materialFunction=&MF_CreativeMoveTool}
  MaterialFunctionCall() {materialFunction=/Game/Global/Base4Layers_2022/MaterialFunctions/MF_GigaMix_2023}
  Multiply(A=@n43) {constB=0.35}
  n83 = NamedRerouteUsage()
  n84 = Add(A=Time(), B=Divide(A=NamedRerouteUsage()))
  n85 = Multiply(
   A=Multiply(A=@n84, B=ScalarParameter() {defaultValue=0.3, parameterName=ScreenEffect.Speed}) {constB=0.1}
   B=ScalarParameter() {defaultValue=0.3, parameterName=ScreenEffect.RandomOffsetSpeed}
  )
  n86 = Sine(None=@n85) {period=0.55}
  n87 = Sine(None=@n85) {period=0.33}
  n88 = Add(A=Add(A=ConstantBiasScale(None=@n86), B=ConstantBiasScale(None=@n87)), B=ConstantBiasScale(None=Sine(None=@n85) {period=64}))
  n89 = Add(
   A=Multiply(A=@n84, B=ScalarParameter() {defaultValue=0.3, parameterName=ScreenEffect.Speed}) {constB=0.1}
   B=Multiply(A=Add(A=@n88) {constB=2}, B=ScalarParameter() {defaultValue=0.2, parameterName=ScreenEffect.OffsetIntensity}) {constB=0.2}
  )
  n90 = ComponentMask(None=Frac(None=ActorPositionWS())) {g=true, r=true}
  n91 = TextureSample!(UVs=Add(A=ComponentMask(None=@n83) {g=true}, B=Add(A=@n89, B=@n90)), Tex, Apply View MipBias) {samplerType=SAMPLERTYPE_LinearGrayscale, texture=/Game/Global/Textures/Default/Masks_and_Gradients/T_ScreenRamp_H}
  n92 = Time()
  n93 = Sine(None=@n92)
  n94 = Sine(None=@n92) {period=0.45}
  n95 = Saturate(None=Add(A=Add(A=@n93, B=@n94), B=Sine(None=@n92) {period=1.63}))
  NamedRerouteDeclaration(None=Multiply(A=Multiply(A=@n91.R, B=@n95), B=ScalarParameter() {defaultValue=0.2, parameterName=ScreenEffect.BumpOffsetIntensity}) {constB=0.2}) {name=ScanlineBumpOffsetHeight, nodeColor=<rgba 0.6875 1 0 1>}
  n96 = TextureObjectParameter() {
   desc="SFX TextureGrayscale You can reuse that texture across layers and functions. "
   group=0_SFX Global
   parameterName=SFX_Grayscale
   samplerType=SAMPLERTYPE_LinearGrayscale
   sortPriority=0
   texture=/Game/Global/Base4Layers_2022/Textures/T_SFX_Grayscale
  }
  n97 = TextureObjectParameter() {
   desc="SFX UniversalTextureMask Uses Dot to select channel used for various parts of material.You can reuse that texture across layers and functions. "
   group=0_SFX Global
   parameterName=SFX_Mask
   samplerType=SAMPLERTYPE_Masks
   sortPriority=0
   texture=/Game/Global/Base4Layers_2022/Textures/T_SFX_M
  }
  n98 = StaticSwitchParameter(True=@n96, False=@n97) {desc=Make sure size is less than 512 and lower, group=0_SFX Global, parameterName=UseSFXGrayscale}
  NamedRerouteDeclaration(None=MaterialFunctionCall!(In=@n98) {materialFunction=/Game/Global/Base4Layers_2022/MaterialFunctions/MF_SFXTexture}.SFX) {name=SFX V3, nodeColor=<rgba 0.6406 0 1 1>}
  n99 = NamedRerouteUsage()
  n100 = MaterialFunctionCall(Float2=@n99) {materialFunction=&BreakOutFloat2Components}
  n101 = MaterialFunctionCall(Float2=OneMinus(None=@n99)) {materialFunction=&BreakOutFloat2Components}
  n102 = Multiply(A=Multiply(A=@n100.R, B=@n100.R), B=Multiply(A=@n101.R, B=@n101.R))
  n103 = OneMinus(None=Saturate(None=Multiply(A=@n102) {constB=40}))
  n104 = Round(None=Multiply(A=Power(Base=@n103) {constExponent=16}) {constB=2})
  n105 = Divide(A=@n104)
  n106 = ScalarParameter() {defaultValue=0.2, parameterName=EdgeOpacity}
  NamedRerouteDeclaration(None=Multiply(A=Saturate(None=@n105), B=@n106) {constB=0.1}) {name=EdgeValues, nodeColor=<rgba 1 0.0234 0 1>}
  n107 = NamedRerouteUsage()
  n108 = Multiply(A=Add(A=Time(), B=Multiply(A=@n107) {constB=100}), B=ScalarParameter() {defaultValue=2, parameterName=Brightness.FlickerSpeed}) {constB=0.1}
  n109 = Sine(None=@n108) {period=0.57}
  n110 = Sine(None=@n108) {period=0.33}
  n111 = Add(A=Add(A=ConstantBiasScale(None=@n109), B=ConstantBiasScale(None=@n110)), B=ConstantBiasScale(None=Sine(None=@n108) {period=1.11}))
  NamedRerouteDeclaration(None=LinearInterpolate(A=ScalarParameter() {defaultValue=0.5, parameterName=Brightness.FlickerMin}, Alpha=Saturate(None=@n111)) {constA=0.5}) {name=BrightnessFlicker, nodeColor=<rgba 0 0.2969 1 1>}
  n112 = MaterialFunctionCall(Float3=ActorPositionWS()) {materialFunction=&BreakOutFloat3Components}
  n113 = Add(A=@n112.R, B=@n112.R)
  NamedRerouteDeclaration(None=Frac(None=Add(A=@n113, B=@n112.B))) {name=Seed, nodeColor=<rgba 0.3828 1 0 1>}
  n114 = NamedRerouteDeclaration(None=TextureCoordinate()) {name=TexCoord, nodeColor=<rgba 0.0547 1 0 1>}
  n115 = Constant()
  NamedRerouteDeclaration(None=If(A=ComponentMask(None=@n114) {r=true}, "A > B"=Constant() {r=1}, "A == B"=@n115, "A < B"=@n115) {constB=1}) {name=UV2_ScreenMask, nodeColor=<rgba 0 1 0.7422 1>}
  n116 = Add(
   A=ComponentMask(None=NamedRerouteUsage()) {g=true}
   B=Multiply(A=Time(), B=ScalarParameter() {defaultValue=0.04, parameterName=Scanlines.AnimSpeed}) {constB=0.04}
  )
  n117 = ScalarParameter() {defaultValue=50, parameterName=Scanlines.Amount}
  n118 = ConstantBiasScale(None=Sine(None=Multiply(A=@n116, B=@n117) {constB=50}))
  n119 = ScalarParameter() {defaultValue=0.01, parameterName=Scanlines.Opacity}
  n120 = Power(Base=Multiply(A=@n91.R, B=LinearInterpolate(Alpha=@n95) {constA=0.4}))
  n121 = ScalarParameter() {defaultValue=1, parameterName=ScreenEffect.Opacity}
  NamedRerouteDeclaration(None=Add(A=Multiply(A=@n118, B=@n119) {constB=0.01}, B=Multiply(A=@n120, B=@n121) {constB=0.4})) {name=ScanlineOut, nodeColor=<rgba 0 1 0.3438 1>}
  n122 = TextureObjectParameter() {
   group=1_Layer_Nanite
   parameterName="Normals  Nanite"
   samplerType=SAMPLERTYPE_Normal
   sortPriority=0
   texture=/Engine/EngineMaterials/FlatNormal
  }
  n123 = TextureObjectParameter() {
   group=1_Layer_Nanite
   parameterName="SpecularMasks  Nanite"
   samplerType=SAMPLERTYPE_Masks
   sortPriority=0
   texture=&T_Fortnite_Default_S
  }
  n124 = TextureObjectParameter() {group=1_Layer_Nanite, parameterName=Diffuse Nanite, sortPriority=0, texture=/Game/Global/Textures/Default/Blanks/T_MidgreysRGB}
  n125 = StaticBoolParameter() {desc="Enables SFX for Opacity Mask, mast have Masked blending set in MI", group=1_Layer SFX, parameterName=SFX OpacityMask, sortPriority=1}
  n126 = TextureObjectParameter() {group=1_Layer, parameterName=Diffuse, sortPriority=0, texture=&T_MidgreysRGB}
  n127 = VectorParameter() {desc=Use R or G or B to select your SFX Mask Channel, group=0_SFX BP, parameterName=SFX_Tint_Ch, sortPriority=4, defaultValue=<rgba 1 0 0 0>}
  n128 = TextureObjectParameter() {
   desc="R = Specular\r\nG = Metal\r\nB = Roughness"
   group=1_Layer
   parameterName=SpecularMasks
   samplerType=SAMPLERTYPE_Masks
   sortPriority=0
   texture=&T_Fortnite_Default_S
  }
  n129 = TextureObjectParameter() {group=1_Layer, parameterName=Normals, samplerType=SAMPLERTYPE_Normal, sortPriority=0, texture=&FlatNormal}
  n130 = VectorParameter() {
   group=1_Layer SFX
   parameterName="SFX_SSS Tint(Expand)"
   sortPriority=0
   channelNames {a=SSS Depth 0-1, b="", g="", r=""}
   defaultValue=<rgba 0 0 0 0.5>
  }
  n131 = VectorParameter() {
   desc=Use R and G channels of this input to controll ClearCoat amount and Rounghness. It uses SFX texture to mask area affected.
   group=1_Layer SFX
   parameterName="SFX_ClearCoat(Expand)"
   sortPriority=-3
   channelNames {a=Not Used, b=Not Used, g=Clearcoat Roughness, r=Clearcoat Value}
   defaultValue=<rgba 1 0 0 0>
  }
  n132 = ScalarParameter() {desc=If you need to cancel shared Mask texture on specific layer, group=1_Layer SFX, parameterName=SFX_OpacityMaskCancel, sortPriority=3}
  n133 = VectorParameter() {
   desc="Emissive color, by default mult by Base color. You can control that blending via Alpha channel of this input. Uses SFX texture as the Mask"
   group=1_Layer SFX
   parameterName="SFX_Emissive_Tint(Expand)"
   sortPriority=-6
   channelNames {a=Base color to Emissive, b="", g="", r=""}
   defaultValue=<rgba 0 0 0 1>
  }
  n134 = ScalarParameter() {
   bUseCustomPrimitiveData=true
   desc="For now Scalar, this is possible alternative to Texture Swatch. Uses single ramp texture shared between all 4 layers and can have indepent row selected per layer. Enabled globally via function. Will disable use of color swatch if enabled. Diffuse tint still can be used if needed via separate mask "
   group=Y_CPD
   parameterName=CPD_PackHSV_1
   primitiveDataIndex=30
   sliderMin=1
  }
  NamedRerouteDeclaration(None=MaterialFunctionCall!(
    WSM Scale
    Use WS Mapping
    Use Z
    Use XYZ
    WSM_ProjectionTransitionContrast
    UV2 True
    WSM_HeightLerpUsingAlpha
    WSM_HeightLerpUsingAlpha_Contrast
    UV's=MaterialFunctionCall!(L3 UV2, L2 UV2, L4 UV2) {materialFunction=/Game/Global/Base4Layers_2022/MaterialFunctions/MF_UVsSet_New}.UVs Layer 1
    Normals_NN=@n122
    SpecularMasks_NN=@n123
    Diffuse_NN=@n124
    Use SFX Emissive=StaticBoolParameter() {desc=Enables SFX, group=1_Layer SFX, parameterName=SFX Emissive, sortPriority=-8}
    SFX Opacity Mask=@n125
    SFX SSS Two Sided Foliange=StaticBoolParameter() {group=1_Layer SFX, parameterName=SFX_SSS Foliage, sortPriority=-2}
    SFX SSS=StaticBoolParameter() {group=1_Layer SFX, parameterName=SFX_SSS, sortPriority=-2}
    SFX Clearcoat=StaticBoolParameter() {group=1_Layer SFX, parameterName=SFX_Clearcoat, sortPriority=-5}
    Diffuse=NamedRerouteDeclaration(None=@n126) {name=L1D, nodeColor=<rgba 0 0.2969 1 1>}
    World Position Z=Constant() {r=1}
    SFX Texture Complete=NamedRerouteUsage()
    SFX Tint Channel=@n127.RGB
    SpecularMasks=NamedRerouteDeclaration(None=@n128) {name=L1S, nodeColor=<rgba 1 0.7734 0 1>}
    Normals=NamedRerouteDeclaration(None=@n129) {name=L1N, nodeColor=<rgba 1 0.3984 0 1>}
    Normals Power=ScalarParameter() {defaultValue=1, group=1_Layer, parameterName=NormalsValue, sortPriority=1}
    "SFX  SSS Channel"=VectorParameter() {group=1_Layer SFX, parameterName=SFX_SSS_Ch, sortPriority=-1}.RGB
    SSS Color Value=AppendVector(A=@n130, B=@n130)
    "Clearcoat (R) CC Roughness (G)"=@n131
    Clearcoat Mask Channal=VectorParameter() {group=1_Layer SFX, parameterName=SFX_Clearcoat_Chl, sortPriority=-4}.RGB
    SFX Opacilty Mask Channel=VectorParameter() {group=1_Layer SFX, parameterName=SFX_OpacityMask_Ch_1, sortPriority=2}.RGB
    Opacilty Mask Cancel=@n132
    "SFX Emissive  Channel"=VectorParameter() {group=1_Layer SFX, parameterName=SFX_Emissive_Ch, sortPriority=-7}.RGB
    Emissive Tint=AppendVector(A=@n133, B=@n133)
    Data Ramp RGB=MaterialFunctionCall!(In=@n134) {materialFunction=/Game/Global/Base4Layers_2022/MaterialFunctions/MF_ColorTint_Unpack}.Result
    Output Diffuse Only=StaticBoolParameter() {group=1_Layer, parameterName=Layer 1 Diffuse Only}
    Output Normals Only
   ) {materialFunction=/Game/Global/Base4Layers_2022/MaterialFunctions/MF_CombineLayers}.Result) {name=L1, nodeColor=<rgba 0 0.5078 1 1>}
  n135 = StaticBoolParameter() {
   desc=Enable to use UV2 for mapping. You may need to switch masks for ohter layers To UV3
   group=2_layer UV
   parameterName=Use UV2 for Layer 2
  }
  n136 = MaterialFunctionCall(L2 UV2=@n135) {materialFunction=&MF_UVsSet_New}
  n137 = StaticBoolParameter() {
   desc="World Space Z only, you need appropriate mask to make it visible. Use Z or map your layers accordingly "
   group=2_layer UV
   parameterName=Layer 2 WSM_Z
  }
  n138 = MaterialFunctionCall(X=Multiply(A=NamedRerouteUsage(), B=ScalarParameter() {defaultValue=0.025, parameterName=Scanlines.TextureWarp}) {constB=0.05}) {materialFunction=&MakeFloat2}
  n139 = Time()
  n140 = Add(A=Sine(None=@n139), B=Sine(None=@n139) {period=0.45})
  n141 = Sine(None=@n139) {period=1.63}
  n142 = ScalarParameter() {defaultValue=100, parameterName=ScreenResolution}
  n143 = Multiply(A=@n136.UVs Layer 2, B=@n142) {constB=100}
  n144 = If(A=Saturate(None=Add(A=@n140, B=@n141)), "A > B"=Divide(A=Ceil(None=@n143), B=@n142) {constB=100}, "A < B"=@n136.UVs Layer 2) {constB=0.95}
  n145 = TextureObjectParameter() {
   group=2_Layer_Nanite
   parameterName="Normals  Nanite 2"
   samplerType=SAMPLERTYPE_Normal
   sortPriority=0
   texture=&FlatNormal
  }
  n146 = TextureObjectParameter() {
   group=2_Layer_Nanite
   parameterName="SpecularMasks  Nanite 2"
   samplerType=SAMPLERTYPE_Masks
   sortPriority=0
   texture=&T_Fortnite_Default_S
  }
  n147 = TextureObjectParameter() {group=2_Layer_Nanite, parameterName=Diffuse Nanite 2, sortPriority=0, texture=&T_MidgreysRGB}
  n148 = ScalarParameter() {
   bUseCustomPrimitiveData=true
   desc="For now Scalar, this is possible alternative to Texture Swatch. Uses single ramp texture shared between all 4 layers and can have indepent row selected per layer. Enabled globally via function. Will disable use of color swatch if enabled. Diffuse tint still can be used if needed via separate mask "
   group=Y_CPD
   parameterName=CPD_PackedHSV_2
   primitiveDataIndex=31
   sliderMin=1
  }
  n149 = StaticBool()
  n150 = StaticBoolParameter() {group=2_Layer_Blending, parameterName=Blend Overlay Layer 2}
  n151 = StaticSwitch(True=StaticBoolParameter() {group=2_Layer_Blending, parameterName=L2 Simple Swatch, sortPriority=-15}, False=@n149, Value=@n150)
  MaterialExpressionReroute_25 = Reroute(None=@n151)
  n152 = StaticSwitch(True=StaticBoolParameter() {group=2_Layer_Blending, parameterName=L2 Normals Only, sortPriority=-15}, False=@n149, Value=@n150)
  MaterialExpressionReroute_31 = Reroute(None=@n152)
  n153 = TextureObjectParameter() {group=2_Layer, parameterName=Diffuse_Texture_2, sortPriority=0, texture=&T_MidgreysRGB}
  n154 = VectorParameter() {
   desc=Use R or G or B to select your SFX Mask Channel
   group=0_SFX BP
   parameterName=SFX_Tint_Ch_2
   sortPriority=4
   defaultValue=<rgba 1 0 0 0>
  }
  n155 = TextureObjectParameter() {
   desc="R = Specular\r\nG = Metal\r\nB = Roughness"
   group=2_Layer
   parameterName=SpecularMasks_2
   samplerType=SAMPLERTYPE_Masks
   sortPriority=0
   texture=&T_Fortnite_Default_S
  }
  n156 = TextureObjectParameter() {group=2_Layer, parameterName=Normals_Texture_2, samplerType=SAMPLERTYPE_Normal, sortPriority=0, texture=&FlatNormal}
  n157 = VectorParameter() {
   group=2_Layer SFX
   parameterName="SFX_SSS Tint (Expand)_2"
   primitiveDataIndex=2
   sortPriority=0
   channelNames {a=SSS Depth 0-1, b="", g="", r=""}
   defaultValue=<rgba 0 0 0 0.5>
  }
  n158 = VectorParameter() {
   desc="Use R and G channels of this input to controll ClearCoat amount and Rounghness. It uses SFX texture to mask area affected. "
   group=2_Layer SFX
   parameterName="SFX_ClearCoat(Expand)_2"
   sortPriority=-3
   channelNames {a=Not Used, b=Not Used, g=Clearcoat Roughness, r=Clearcoat Value}
   defaultValue=<rgba 1 0 0 0>
  }
  n159 = ScalarParameter() {desc=If you need to cancel shared Mask texture on specific layer, group=2_Layer SFX, parameterName=SFX_OpacityMaskCancel_2, sortPriority=3}
  n160 = VectorParameter() {
   desc="Emissive color, by default mult by Base color. You can control that blending via Alpha channel of this input. Uses SFX texture as the Mask"
   group=2_Layer SFX
   parameterName="SFX_Emissive_Tint(Expand)_2"
   sortPriority=-6
   channelNames {a=Base color to Emissive, b="", g="", r=""}
   defaultValue=<rgba 0 0 0 1>
  }
  NamedRerouteDeclaration(None=MaterialFunctionCall(
    WSM Scale=@n136.WSM 2
    Use WS Mapping=StaticBoolParameter() {desc=Enable to use World Space Mapping for layer, group=2_layer UV, parameterName=Layer 2 WS Mapping}
    Use Z=@n137
    Use XYZ=StaticBoolParameter() {desc="World Space XYZ ( Expensive)", group=2_layer UV, parameterName=Layer 2 WSM_XYZ}
    UV2 True=@n135
    UV's=Add(A=@n138.Result, B=@n144)
    Normals_NN=@n145
    SpecularMasks_NN=@n146
    Diffuse_NN=@n147
    Use SFX Emissive=StaticBoolParameter() {desc=Enables SFX, group=2_Layer SFX, parameterName=SFX Emissive 2, sortPriority=-8}
    SFX Opacity Mask=StaticBoolParameter() {desc=Enables SFX, group=2_Layer SFX, parameterName=SFX OpacityMask 2, sortPriority=1}
    SFX SSS Two Sided Foliange=StaticBoolParameter() {group=2_Layer SFX, parameterName=SFX_SSS Foliage 2, sortPriority=-2}
    SFX SSS=StaticBoolParameter() {group=2_Layer SFX, parameterName=SFX_SSS 2, sortPriority=-2}
    SFX Clearcoat=StaticBoolParameter() {group=2_Layer SFX, parameterName="SFX_Clearcoat  2", sortPriority=-5}
    Diffuse=NamedRerouteDeclaration(None=@n153) {name=D2, nodeColor=<rgba 1 0.3984 0 1>}
    World Position Z=ScalarParameter() {group=2_layer UV, parameterName=L2 WP Offset Z, sortPriority=1}
    SFX Texture Complete=NamedRerouteUsage()
    SFX Tint Channel=@n154.RGB
    SpecularMasks=NamedRerouteDeclaration(None=@n155) {name=S2, nodeColor=<rgba 1 0 0.3281 1>}
    Normals=NamedRerouteDeclaration(None=@n156) {name=N2, nodeColor=<rgba 0 0.5547 1 1>}
    Normals Power=ScalarParameter() {defaultValue=1, group=2_Layer, parameterName=NormalsValue_2, sortPriority=1}
    "SFX  SSS Channel"=VectorParameter() {group=2_Layer SFX, parameterName=SFX_SSS_Ch_2, sortPriority=-1}.RGB
    SSS Color Value=AppendVector(A=@n157, B=@n157)
    "Clearcoat (R) CC Roughness (G)"=@n158
    Clearcoat Mask Channal=VectorParameter() {group=2_Layer SFX, parameterName=SFX_Clearcoat_Ch_2, sortPriority=-4}.RGB
    SFX Opacilty Mask Channel=VectorParameter() {group=2_Layer SFX, parameterName=SFX_OpacityMask_Ch_2, sortPriority=2}.RGB
    Opacilty Mask Cancel=@n159
    "SFX Emissive  Channel"=VectorParameter() {group=2_Layer SFX, parameterName=SFX_Emissive_Ch_2, sortPriority=-7}.RGB
    Emissive Tint=AppendVector(A=@n160, B=@n160)
    Data Ramp RGB=MaterialFunctionCall(In=@n148) {materialFunction=&MF_ColorTint_Unpack}.Result
    Output Diffuse Only=@MaterialExpressionReroute_25.Output
    Output Normals Only=@MaterialExpressionReroute_31.Output2
   ) {materialFunction=&MF_CombineLayers}.Result) {name=L2, nodeColor=<rgba 1 0 0.5859 1>}
  NamedRerouteDeclaration(None=@n98) {name=SFX Texture Global, nodeColor=<rgba 1 0.3516 0 1>}
  n161 = TextureObjectParameter() {
   desc="SFX RGB - Graffiti, emissive RGB "
   group=0_SFX Global
   parameterName=SFX_RGB
   sortPriority=1
   texture=/Game/Global/Base4Layers_2022/Textures/T-GraffitiSample_D
  }
  NamedRerouteDeclaration(None=@n161) {name=SFX Texture RGB, nodeColor=<rgba 0 0.5312 1 1>}
  n162 = MaterialFunctionCall(Float2=@n90) {materialFunction=&BreakOutFloat2Components}
  NamedRerouteDeclaration(None=Divide(A=Add(A=@n162.R, B=@n162.R))) {name=RandomAlpha, nodeColor=<rgba 1 0 0.5625 1>}
  NamedRerouteDeclaration(None=Multiply(A=Saturate(None=@n104), B=@n106) {constB=0.1}) {name=EdgeMask, nodeColor=<rgba 0 1 0.1094 1>}
  NamedRerouteUsage()
  PinBase() {reroutePins [Input {expression=&self:MaterialExpressionReroute_3}]}
  PinBase!(Output=@n70) {reroutePins [Output {expression=&self:MaterialExpressionReroute_2}]}
  PinBase!()
  PinBase!(Output=@n151, Output2=@n152) {reroutePins [
    Output {expression=&self:MaterialExpressionReroute_25}
    Output2 {expression=&self:MaterialExpressionReroute_31}
   ]}
  PinBase!() {reroutePins [
    "For Moss (MA)" {expression=&self:MaterialExpressionReroute_8}
    "RVT_Height (S)" {expression=&self:MaterialExpressionReroute_1}
   ]}
  PinBase!(Result=@n72.Buildings Jungl) {reroutePins [Result {expression=&self:MaterialExpressionReroute_0}]}
  PinBase!() {reroutePins [MaterialAttributes {expression=&self:MaterialExpressionReroute_70}]}
  PinBase!(Output=@n64) {reroutePins [Output {expression=&self:MaterialExpressionReroute_71}]}
  PinBase!() {reroutePins [MaterialAttributes {expression=&self:MaterialExpressionReroute_69}]}
  PinBase!(Result=@n66.Result) {reroutePins [Result {expression=&self:MaterialExpressionReroute_68}]}
  PinBase!() {reroutePins [Input {expression=&self:MaterialExpressionReroute_12}]}
  PinBase!(Output=@n14) {reroutePins [Output {expression=&self:MaterialExpressionReroute_14}]}
  PinBase!() {reroutePins [InputPin {expression=&self:MaterialExpressionReroute_7}]}
  PinBase!(Output=@n82) {reroutePins [Output {expression=&self:MaterialExpressionReroute_6}]}
  Reroute(None=@n1)
  Sine(None=Time())
 ]
}
