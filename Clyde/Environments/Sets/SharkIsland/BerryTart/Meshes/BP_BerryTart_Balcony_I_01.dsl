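; Construction script for a BerryTart balcony Blueprint in the Shark Island environment set, controlling the strength of its four Lumen plane components.
; It reads the LumenPlaneStrength variable and writes it as default custom primitive data float at index 0 on LumenPlane0 through LumenPlane3.
; One variable thus drives all four planes through per-primitive data, with no dynamic material instances created in the script.
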(fn ConstructionScript
  (bind _lumenplanestrength (Variables|Default|GetLumenPlaneStrength))
  (Rendering|Material|SetDefaultCustomPrimitiveDataFloat (Variables|Default|GetLumenPlane0) 0 _lumenplanestrength)
  (Rendering|Material|SetDefaultCustomPrimitiveDataFloat (Variables|Default|GetLumenPlane1) 0 _lumenplanestrength)
  (Rendering|Material|SetDefaultCustomPrimitiveDataFloat (Variables|Default|GetLumenPlane2) 0 _lumenplanestrength)
  (Rendering|Material|SetDefaultCustomPrimitiveDataFloat (Variables|Default|GetLumenPlane3) 0 _lumenplanestrength))
