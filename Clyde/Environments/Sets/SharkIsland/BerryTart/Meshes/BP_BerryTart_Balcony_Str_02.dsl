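; BP_BerryTart_Balcony_Str_02 is a balcony mesh Blueprint for the BerryTart set on Shark Island, judging by its name and path.
; Its construction script writes the LumenPlaneStrength variable into custom primitive data slot 0 of the LumenPlane component.
; The material on that plane can read this per-instance float, presumably to tune its contribution to Lumen lighting.
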(fn ConstructionScript
  (Rendering|Material|SetDefaultCustomPrimitiveDataFloat (Variables|Default|GetLumenPlane) 0 (Variables|Default|GetLumenPlaneStrength)))
